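# Translucent UI material for an animated panel background, with panning noise over a two-color halftone look.
# Three panning cell-noise samples combine with an inverted sphere-mask gradient to blend Base Color into Secondary Color multiplied by a UV-scaled halftone pattern.
# Opacity comes from a width-adjustable sphere mask whose coordinates are distorted by the same noise, giving a moving edge.
# Named reroutes carry Color and Opacity, and parameters expose colors, halftone, pan speed and mask shaping.
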Material /PistolCanary/Layout/Material/Boards/M_PanelBG_Panning
graph {
 properties {bAutomaticallySetUsageInEditor=false, blendMode=BLEND_Translucent, materialDomain=MD_UI, usedShadingModels=Default Lit}
 nodes [
  Add(B=Sine(None=Time()) {period=2})
  n1 = VectorParameter() {group=Color, parameterName=Secondary Color, defaultValue=<rgba 0.3021 0.0397 0.178 1>}
  n2 = MaterialFunctionCall!(UV=TextureCoordinate(), TextureSize, ScalePivotPoint) {materialFunction=/FortUILibrary/UGC/MaterialFunctions/MF_UGCUI_UVScaling}
  n3 = ComponentMask(None=VectorParameter() {group=HalfToneControl, parameterName=HalfToneScale, defaultValue=<rgba 50 50 0 0>}.RGB) {g=true, r=true}
  n4 = Power(
   Base=TextureSample(UVs=Multiply(A=@n2.Scaled to Fit, B=@n3)) {samplerType=SAMPLERTYPE_LinearGrayscale, texture=/Game/Creative/PostProcess/Textures/T_PP_Halftone_Pattern}.R
   Exp=ScalarParameter() {defaultValue=2, group=HalfToneControl, parameterName=HalftoneContrast}
  ) {constExponent=3}
  n5 = TextureCoordinate() {vTiling=0.05}
  n6 = ComponentMask(None=VectorParameter() {group=NoiseControl, parameterName=PanSpeed, defaultValue=<rgba 0.1 1 0 0>}.RGB) {g=true, r=true}
  n7 = Add(A=TextureCoordinate() {uTiling=12, vTiling=0.025}) {constB=0.3}
  n8 = Multiply(
   A=TextureSample(UVs=Panner(Coordinate=@n5, Speed=@n6) {speedY=1}) {texture=/Game/Creative/PostProcess/Textures/Cell_Noise_T1}.G
   B=TextureSample(UVs=Panner(Coordinate=@n7, Speed=@n6) {speedY=1}) {texture=&Cell_Noise_T1}.G
  )
  n9 = Panner(Coordinate=Add(A=TextureCoordinate() {uTiling=0.05, vTiling=0.15}) {constB=0.73}, Speed=@n6) {speedY=1}
  n10 = Power(Base=Multiply(A=TextureSample(UVs=@n9) {texture=&Cell_Noise_T1}.G) {constB=2.5})
  n11 = TextureCoordinate() {vTiling=0}
  n12 = ComponentMask(None=@n11) {r=true}
  n13 = ScalarParameter() {defaultValue=1, group=NoiseControl, parameterName=GradientControl}
  n14 = SphereMask(
   A=AppendVector(A=Multiply(A=@n12, B=@n13), B=ComponentMask(None=@n11) {g=true})
   B=AppendVector(A=Multiply(A=@n13) {constB=0.5}, B=Constant2Vector().G)
  ) {attenuationRadius=0.5, hardnessPercent=0}
  n15 = Power(Base=Add(A=Multiply(A=@n8, B=@n10), B=OneMinus(None=@n14)), Exp=ScalarParameter() {defaultValue=2, group=NoiseControl, parameterName=Power}) {constExponent=1}
  NamedRerouteDeclaration(None=LinearInterpolate(
    A=VectorParameter() {group=Color, parameterName=Base Color, defaultValue=<rgba 1 0.1244 0.8582 1>}.RGB
    B=Multiply(A=@n1.RGB, B=@n4)
    Alpha=Saturate(None=@n15)
   ) {constAlpha=0}) {name=Color, nodeColor=<rgba 1 0 0.1406 1>}
  n16 = Multiply(A=@n8) {constB=0.1}
  n17 = MaterialFunctionCall(Float2=TextureCoordinate()) {materialFunction=&BreakOutFloat2Components}
  n18 = ScalarParameter() {defaultValue=0.25, group=Opacity, parameterName=OpacityMaskWidth}
  n19 = AppendVector(A=Multiply(A=@n17.R, B=@n18), B=@n17.G)
  n20 = Constant2Vector() {g=0.5, r=1}
  n21 = Multiply(A=Multiply(A=@n18) {constB=0.5}, B=@n20.R)
  n22 = Add(A=@n20.G) {constB=0.01}
  NamedRerouteDeclaration(None=SphereMask(A=Add(A=@n16, B=@n19), B=AppendVector(A=@n21, B=@n22)) {attenuationRadius=0.5, hardnessPercent=85}) {name=Opacity, nodeColor=<rgba 0.8984 0 1 1>}
  NamedRerouteUsage()
  NamedRerouteUsage()
 ]
}
