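# Default Lit material in the PistolCanary layout's Boards set, named tac3000, apparently intended for a board surface.
# Its graph holds a single Texture Sample of the PistolCanary_Boards_0106 texture, with no connection shown to any material input.
# No parameters are exposed, and automatic setting of usage flags in the editor is turned off.
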Material /PistolCanary/Layout/Material/Boards/PistolCanary_tac3000_Mat_001
graph {
 properties {bAutomaticallySetUsageInEditor=false, usedShadingModels=Default Lit}
 nodes [TextureSample() {texture=/PistolCanary/Layout/Texture/Boards/PistolCanary_Boards_0106}]
}
